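; Season 24 quest prop: a shooting target that flips down only for the player who validly hits it.
; On server damage, it requires an unfinished quest objective, required pawn tags and a matching weapon tag before flipping and calling parent damage logic.
; The multicast flip acts only on the instigator's client, splitscreen included, toggling ambient effects and rotating the mesh by timeline.
; Show-target overlap events are empty, and construction removes the discover-radius collision.
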(event OnComponentBeginOverlap(Collision_ShowTarget)
  (params (OverlappedComponent :type Primitive%20Component%20Object%20Reference) (OtherActor :type Actor%20Object%20Reference) (OtherComp :type Primitive%20Component%20Object%20Reference) (OtherBodyIndex :type Integer) (bFromSweep :type Boolean) (SweepResult :type Hit%20Result%20Structure%20%28by%20ref%29)))

(event OnComponentEndOverlap(Collision_ShowTarget)
  (params (OverlappedComponent :type Primitive%20Component%20Object%20Reference) (OtherActor :type Actor%20Object%20Reference) (OtherComp :type Primitive%20Component%20Object%20Reference) (OtherBodyIndex :type Integer)))

(event Custom|ClientFlipTarget_Replicated
  (params (bFlipUp :type Boolean) (InstigatingPlayerController :type Fort%20Player%20Controller%20Object%20Reference))
  (bind _n_ambient (Variables|Components|GetN_Ambient))
  ;[0 0 0] Multicast - Flip the target up or down for the instigating player
  (Networking|SwitchHasAuthority
    (:Remote
      (bind _output (CallFunction|IsValidLocalPlayerController InstigatingPlayerController))
      (if _output
        (if bFlipUp
          (Components|Activation|Activate _n_ambient true)
          (Variables|Default|SetLerpStart -9e+01)
          (Variables|Default|SetLerpEnd 0.0)
          (else
            (Components|Activation|Deactivate _n_ambient)
            (Variables|Default|SetLerpStart 0.0)
            (Variables|Default|SetLerpEnd -9e+01)))
        (bind (_direction _fliplerptrack) (Timeline|FlipTimeline PlayFromStart)
          (:Update
            (Transformation|SetRelativeRotation (Variables|Components|GetSM_Quest_PreInteraction) (Math|Rotator|MakeRotator (Math|Float|Lerp (Variables|Default|GetLerpStart) (Variables|Default|GetLerpEnd) _fliplerptrack)) false false)))))))

(event EventOnDamageServer
  (params (Damage :type Float%20%28single-precision%29) (DamageTags :type Gameplay%20Tag%20Container%20Structure%20%28by%20ref%29) (Momentum :type Vector :default "0, 0, 0") (HitInfo :type Hit%20Result%20Structure%20%28by%20ref%29) (InstigatedBy :type Controller%20Object%20Reference) (DamageCauser :type Actor%20Object%20Reference) (EffectContext :type Gameplay%20Effect%20Context%20Handle%20Structure))
  ;[0 0 0] Server only - logic for when a player hits a target
  ;;[0.1719 0.1719 0.1719] Ensure that a player hit the target, and that they have the quest and objective assoicated with this target
  (bind _asfort_player_controller (Utilities|Casting|CastToFortPlayerController InstigatedBy))
  (bind _bcanplay (CallFunction|InstigatorCanPlay _asfort_player_controller))
  (if _bcanplay
    (if (and (GameplayTags|HasAllTags (GameplayTags|GetOwnedGameplayTags (Game|Player|GetPlayerPawn _asfort_player_controller) :!class /Script/GameplayTags.GameplayTagAssetInterface) (Variables|Default|GetRequiredTagsToComplete) true) (GameplayTags|HasAnyTags DamageTags (Variables|Default|GetWeaponTags)))
      ;;[0.1719 0.1719 0.1719] Flip target down for that player only
      (CallFunction|ClientFlipTargetReplicated false _asfort_player_controller)
      (|CallParentFunction Damage DamageTags Momentum HitInfo InstigatedBy DamageCauser EffectContext))))

(fn InstigatorCanPlay
  (params (InstigatingPlayerController :type Fort%20Player%20Controller%20Object%20Reference))
  (bind _consolidatedquestcomponent (Variables|BuildingGameplayActorPropQuest|GetConsolidatedQuestComponent))
  (bind _returnvalue (Class|ConsolidatedQuestComponent|GetQuestItemDefinition _consolidatedquestcomponent))
  ;[0 0 0] Check if this player is on the quest and they have not already completed the objective for this target
  (Utilities|IsValid InstigatingPlayerController
    (:Is%20Valid
      (Utilities|IsValid _returnvalue
        (:Is%20Valid
          (Utilities|IsValid (Quests|GetQuestManager InstigatingPlayerController "Invalid")
            (:Is%20Valid
              (bind _playerhasquest (Class|AthenaQuestFunctionLibrary|AthenaHasQuest InstigatingPlayerController _returnvalue))
              (if _playerhasquest
                (bind (_playercontrollerout _completedquest) (Quest|HasPlayerCompletedQuest InstigatingPlayerController _returnvalue (Variables|/Script/FortniteGame.ConsolidatedQuestComponent|GetObjectiveBackendName (Variables|BuildingGameplayActorPropQuest|GetConsolidatedQuestComponent))))
                (if (not _completedquest)
                  (return true)
                  (else
                    _))
                (else
                  _))
              (return false))
            (:Is%20Not%20Valid
              _)))
        (:Is%20Not%20Valid
          _)))
    (:Is%20Not%20Valid
      _)))

(fn IsValidLocalPlayerController
  (params (InstigatingPlayerController :type Fort%20Player%20Controller%20Object%20Reference))
  ;[0 0 0] Returns true on the executing client if the instigator controller is locally controlled. Returns false if running on the server
  (if (not (Networking|IsServer))
    (Utilities|IsValid InstigatingPlayerController
      (:Is%20Valid
        (if (PlayerController|IsinSplitscreen InstigatingPlayerController)
          (for _array_element (Fortnite|GetLocalFortPlayerControllers true)
            (if (== _array_element InstigatingPlayerController)
              (return true)))
          (return false)
          (else
            (return true))))
      (:Is%20Not%20Valid
        _)))
  (return false))

(fn PostInteractionVFX&Audio
  (params (PlayerController :type Fort%20Player%20Controller%20Object%20Reference))
  (|CallParentFunction PlayerController))

(fn PostInteractionVisuals
  (params (PlayerController :type Fort%20Player%20Controller%20Object%20Reference))
  (|CallParentFunction PlayerController))

(fn ConstructionScript
  (|CallParentFunction)
  (Components|DestroyComponent (Variables|Components|GetCollision_DiscoverRadius)))
